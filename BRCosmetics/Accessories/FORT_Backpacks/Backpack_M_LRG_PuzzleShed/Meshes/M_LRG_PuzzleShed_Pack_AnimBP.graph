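# Animation Blueprint for the PuzzleShed backpack cosmetic that procedurally adjusts its pack bone to the wearer's posture and state.
# Each update maps spine_05 yaw relative to spine_04 into additive pack roll and offset.
# The AnimGraph blends crouch-moving and skydiving adjustments of the input pose, applies that spine correction, then adds a skydive-float nudge.
# A melee bone chain and a spine_01-based rotator calculation are left unconnected.
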AnimBlueprint /BRCosmetics/Accessories/FORT_Backpacks/Backpack_M_LRG_PuzzleShed/Meshes/M_LRG_PuzzleShed_Pack_AnimBP
target_skeleton=&/M_LRG_PuzzleShed_Pack_Skeleton
graphs [
 AnimGraph {nodes [
   n1 = "Transform (Modify) Bone - Bone: pack"!(
    ComponentPose
    bAlphaBoolEnabled:bool=Get bIsPlayingMeleeAnim()
    Translation="0,0,0"
    Rotation=Get PACK ROTATOR() -> (PACK ROTATOR:Rotator)
    Scale
   ) -> (Pose) {
    alphaInputType=Bool
    rotationMode=BMM_Additive
    scaleMode=BMM_Additive
    scaleSpace=BCS_BoneSpace
    translationMode=BMM_Additive
    alphaBoolBlend {blendInTime=0.2, blendOption=ExpIn, blendOutTime=0.8}
    boneToModify {boneName=pack}
   }
   "Transform (Modify) Bone - Bone: pack"(ComponentPose=@n1, bAlphaBoolEnabled=Get bIsSkydiveFloating(), Translation="0,-0.2,0", Rotation="0,0,5") {
    alphaInputType=Bool
    rotationMode=BMM_Additive
    scaleMode=BMM_Additive
    scaleSpace=BCS_BoneSpace
    translationMode=BMM_Additive
    alphaBoolBlend {blendInTime=0.2, blendOption=Linear, blendOutTime=0.8}
    boneToModify {boneName=pack}
    rotation {roll=5}
    translation {y=-0.2}
   }
   n2 = "Transform (Modify) Bone - Bone: pack"!(
    ComponentPose=Local To Component(LocalPose=Use cached pose 'StartPose'())
    Alpha:float
    Translation="0,0,2"
    Rotation="0,0,10"
    Scale="0.85,0.85,0.85"
   ) -> (Pose) {
    rotationMode=BMM_Additive
    scaleMode=BMM_Replace
    translationMode=BMM_Additive
    alphaBoolBlend {blendOption=Linear}
    boneToModify {boneName=pack}
    rotation {roll=10}
    scale=<0.85 0.85 0.85>
    translation {z=2}
   }
   n3 = Blend Poses by bool(
    BlendPose_0=Component To Local(ComponentPose=@n2)
    BlendPose_1=Use cached pose 'StartPose'()
    BlendTime_0=0.4
    BlendTime_1=0.3
    bActiveValue=Get bIsSkyDiving()
   ) {blendType=Linear}
   n4 = Use cached pose 'StartPose'()
   n5 = Component To Local(ComponentPose="Transform (Modify) Bone - Bone: pack"!(ComponentPose=Local To Component(LocalPose=@n4), Alpha, Rotation="0,0,-6") -> (Pose) {rotationMode=BMM_Additive, alphaBoolBlend {blendOption=Linear}, boneToModify {boneName=pack}, rotation {roll=-6}})
   n6 = Get CrouchMoving() -> (CrouchMoving:bool)
   n7 = "Transform (Modify) Bone - Bone: pack"!(
    ComponentPose=Local To Component(LocalPose=Blend Poses by bool(BlendPose_0=@n5, BlendPose_1=@n3, BlendTime_0=0.5, BlendTime_1=0.5, bActiveValue=@n6) {blendType=Linear})
    Alpha
    Translation:Vector=Get backpack_translation()
    Rotation=Get backpack_rotate()
    Scale:Vector
   ) -> (Pose) {
    rotationMode=BMM_Additive
    scaleMode=BMM_Additive
    scaleSpace=BCS_BoneSpace
    translationMode=BMM_Additive
    alphaBoolBlend {blendOption=Linear}
    boneToModify {boneName=pack}
   }
   n8 = "Transform (Modify) Bone - Bone: pack"!(ComponentPose=@n7, bAlphaBoolEnabled:bool=Get bIsSkydiveFloating(), Translation="0,1,0", Rotation="0,0,-3", Scale) -> (Pose) {
    alphaInputType=Bool
    rotationMode=BMM_Additive
    scaleMode=BMM_Additive
    scaleSpace=BCS_BoneSpace
    translationMode=BMM_Additive
    alphaBoolBlend {blendInTime=0.2, blendOption=Linear, blendOutTime=0.8}
    boneToModify {boneName=pack}
    rotation {roll=-3}
    translation {y=1}
   }
   Output Pose(Result=Component To Local(ComponentPose=@n8)) {layerGroup=None}
   Save cached pose 'StartPose'(Pose=Input Pose()) {cacheName=StartPose}
   Get NewVar() -> (NewVar:Vector)
  ]}
 EventGraph {nodes [
   n1 = Event BlueprintUpdateAnimation()
   > Set Delta Time X(Delta Time X=@n1.DeltaTimeX)
   > n2 = Sequence()
   Set CrouchMoving(execute=@n2.then_1, CrouchMoving:bool=AND Boolean(A=Get bIsCrouching(), B=Get bIsMoving2D())) -> (Output_Get:bool)
   > Set Spine Corr(Spine Corr:bool=OR Boolean(A=Get bIsPlayingMeleeAnim(), B=Get bIsSkydiveFloating())) -> (Output_Get:bool)
   n3 = GetOwnerSkeletalMeshComponent()
   n4 = BreakRotator(InRot=BreakTransform(InTransform=GetDeltaTransformFromRefPose(self=@n3, BoneName=spine_05, BaseName=spine_04)).Rotation)
   Set backpack_rotate(execute=@n2.then_0, backpack_rotate=MakeRotator(Roll=MapRangeClamped(Value=@n4.Yaw, InRangeA=0, InRangeB=-30, OutRangeB=30)))
   > Set backpack_translation(backpack_translation=MakeVector(Y=MapRangeClamped(Value=@n4.Yaw, InRangeA=0, InRangeB=-30, OutRangeA=-1, OutRangeB=-4)))
   Event OnCharacterCustomizationLoaded()
   > Set Body Mesh(Body Mesh:Skeletal Mesh Component Object Reference=GetOwnerSkeletalMeshComponent()) -> (Output_Get:Skeletal Mesh Component Object Reference)
   n5 = GetOwnerSkeletalMeshComponent()
   n6 = BreakRotator(InRot=BreakTransform(InTransform=GetDeltaTransformFromRefPose(self=@n5, BoneName=spine_05, BaseName=spine_01)).Rotation)
   n7 = MapRangeClamped(Value=@n6, InRangeA=0, InRangeB=-36, OutRangeB=-3)
   Set NewVar(NewVar:Vector=MakeVector(Y=MapRangeClamped(Value=@n6, InRangeA=0, InRangeB=-43, OutRangeB=@n7))) -> (Output_Get:Vector)
   n8 = MapRangeClamped(Value=@n6.Pitch, InRangeA=5, InRangeB=-15, OutRangeB=10)
   n9 = MapRangeClamped(Value=@n6.Yaw, InRangeA=5, InRangeB=-15, OutRangeB=30)
   > Set PACK ROTATOR(PACK ROTATOR:Rotator=MakeRotator(Roll=Ease!(Function:EEasingFunc Enum=ExpoInOut, Alpha:double=1, A:double=@n8, B:double=@n9) -> (Result:double=~) {is=title K2Node_EaseFunction})) -> (Output_Get:Rotator)
   TryGetPawnOwner()
   FInterpTo(Current=float + float(A=@n8, B=@n9), Target=6, DeltaTime=Get Delta Time X(), InterpSpeed=0.6)
   Reroute Node(InputPin:float=@n6) -> (OutputPin:float)
  ]}
]
variables [
 backpack_rotate {type=Rotator}
 backpack_translation {type=Vector}
 Body Mesh {type=SkeletalMeshComponent}
 CrouchMoving {type=boolean}
 Delta Time X {type=number}
 NewVar {type=Vector}
 NewVar_0 {type=number}
 PACK ROTATOR {type=Rotator}
 Spine Corr {type=boolean}
]
